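# Environment query that picks an actor as a reposition destination for the TidalCrane ToyFox boss.
# A custom generator supplies candidate actors using the target giant context, then 2D distance filters drop those too near the boss or outside a band around that giant.
# Scoring favors candidates nearer the giant within that band and inside the safe zone, with a negatively weighted delta-distance test against all participant players.
# Key thresholds and weights bind to scalable floats.
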EnvQuery /TidalCrane_ToyFox_Boss/Abilities/Reposition/EQS/EQS_TidalCraneToyFoxRepositionActor
options [{
  generator {
   name=EnvQueryGeneratorTidalCraneToyFoxRepositionActor_C_0
   class=EnvQueryGeneratorTidalCraneToyFoxRepositionActor_C
   properties {
    bAutoSortTests=true
    bCanRunAsync=false
    context=&/EnvQueryTidalCraneToyFoxTargetGiant.EnvQueryTidalCraneToyFoxTargetGiant_C
    envQueryResultNormalizationOption=Default
    generatedItemType=/Script/AIModule.EnvQueryItemType_Actor
    generatorsActionDescription=""
    optionName=""
   }
  }
  tests [
   EnvQueryTest_Distance_0 {
    class=EnvQueryTest_Distance
    properties {
     distanceTo=&EnvQueryContext_Querier
     filterType=Minimum
     floatValueMax=12000
     scoringEquation=Linear
     scoringFactor=1
     testComment=""
     testMode=Distance2D
     testPurpose=Filter
     floatValueMin {
      dataBinding=&self:EnvQueryTest_Distance_0.FortAIDataProvider_ScalableFloat_0
      dataField=FloatValue
      defaultValue=2000
     }
    }
   }
   EnvQueryTest_Distance_1 {
    class=EnvQueryTest_Distance
    properties {
     distanceTo=&EnvQueryTidalCraneToyFoxTargetGiant_C
     filterType=Range
     scoringEquation=Linear
     testComment=""
     testMode=Distance2D
     testPurpose=FilterAndScore
     floatValueMax {
      dataBinding=&self:EnvQueryTest_Distance_1.FortAIDataProvider_ScalableFloat_1
      dataField=FloatValue
      defaultValue=20000
     }
     floatValueMin {
      dataBinding=&self:EnvQueryTest_Distance_1.FortAIDataProvider_ScalableFloat_0
      dataField=FloatValue
      defaultValue=11000
     }
     scoringFactor {
      dataBinding=&self:EnvQueryTest_Distance_1.FortAIDataProvider_ScalableFloat_2
      dataField=FloatValue
      defaultValue=-2
     }
    }
   }
   FortQueryTest_InsideAthenaSafeZone_0 {
    class=FortQueryTest_InsideAthenaSafeZone
    properties {
     testPurpose=Score
     scoringFactor {
      dataBinding=&self:FortQueryTest_InsideAthenaSafeZone_0.FortAIDataProvider_ScalableFloat_0
      dataField=FloatValue
      defaultValue=3
     }
    }
   }
   FortQueryTest_DeltaDistance_0 {
    class=FortQueryTest_DeltaDistance
    properties {
     bDefineReferenceValue=false
     bUseDistance2D=true
     boolValue=true
     clampMaxType=None
     clampMinType=None
     filterType=Range
     floatValueMax=0
     floatValueMin=0
     locationProviderContext=/Game/AI/EQS/Contexts/FortQueryContext_AllAthenaGameParticipantPlayerPawns.FortQueryContext_AllAthenaGameParticipantPlayerPawns_C
     multipleContextFilterOp=AllPass
     multipleContextScoreOp=AverageScore
     normalizationType=Absolute
     referenceValue=0
     scoreClampMax=0
     scoreClampMin=0
     scoringEquation=Linear
     testComment=""
     testPurpose=Score
     scoringFactor {
      dataBinding=&self:FortQueryTest_DeltaDistance_0.FortAIDataProvider_ScalableFloat_0
      dataField=FloatValue
      defaultValue=-1
     }
    }
   }
  ]
 }]
