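; EQS query generator for the TidalCrane ToyFox boss's Reposition ability, producing candidate reposition actors in the world.
; Its item generation asks the Fort global actor registry subsystem for actors registered under the boss's Reposition Actor gameplay tag.
; Each returned actor becomes a generated query item, with no filtering and no use of the context locations.
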(fn DoItemGeneration
  (params (ContextLocations :type Array%20of%20Vectors))
  ;[0 0.15 0.004128 0.5] Get all the reposition actors in the world.
  (Utilities|FlowControl|Sequence
    (:then_0
      (bind _outactors (Class|FortGlobalActorRegistrySubsystem|GetRegisteredActors (WorldSubsystems|GetFortGlobalActorRegistrySubsystem) (GameplayTags|MakeLiteralGameplayTag "(TagName=\"TidalCrane.ToyFox.Boss.Ability.Reposition.Actor\")")))
      (for _array_element _outactors
        (EQS|AddGeneratedActor _array_element)))
    (:then_1
      (return))))
