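; EQS context for the TidalCrane ToyFox boss's reposition ability that provides a single actor: the giant player pawn (toyfox).
; If the querier actor is valid, it gathers all FortGiantPlayerPawn actors and returns the one nearest the querier's location.
; The GetAllActorsOfClass search relies on only one giant being expected in the world.
; A sequence fallback returns no actor when the querier is invalid.
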(fn ProvideSingleActor
  (params (QuerierObject :type Object%20Reference) (QuerierActor :type Actor%20Object%20Reference))
  ;[0.1378 0.01819 0.15 0.5] Return the single giant (toyfox) expected in the world.
  (Utilities|FlowControl|Sequence
    (:then_0
      (Utilities|IsValid QuerierActor
        (:Is%20Valid
          (bind _outactors (Actor|GetAllActorsOfClass "/Script/GiantPawnCoreRuntime.FortGiantPlayerPawn")) ; Should be OK to use since there's only one actor of this class expected.
          (return (.ReturnValue (Actor|FindNearestActor (Transformation|GetActorLocation QuerierActor) _outactors))))))
    (:then_1
      (return 0))))
